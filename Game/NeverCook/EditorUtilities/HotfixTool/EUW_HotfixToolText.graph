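# Editor utility widget in the HotfixTool folder that displays a single block of text in a bordered box.
# A SizeBox with a minimum desired width holds two nested borders, a dark outline around a grey fill, wrapping an auto-wrapping CommonTextBlock.
# SetText takes a string, converts it to text and writes it to that block; the construct and tick events are empty.
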EditorUtilityWidgetBlueprint /Game/NeverCook/EditorUtilities/HotfixTool/EUW_HotfixToolText
graphs [
 EventGraph {nodes [
   Event PreConstruct()
   Event Construct()
   Event Tick()
  ]}
 SetText {nodes [
   n1 = SetText!() -> (then, InText:String) {is=title K2Node_FunctionEntry}
   > Set Text(Text="To Text (String)"(InString=@n1), self=Get Text_Value() -> (Text_Value:Common Text Object Reference))
  ]}
]
variables []
widgets {
 info {rootWidgetClass=&SizeBox, widgetCount=4}
 widgets [
  SizeBox_0 {
   widgetClassPath=&SizeBox
   widget=SizeBox_0
   properties {bOverride_MinDesiredHeight=true, bOverride_MinDesiredWidth=true, minDesiredWidth=250}
  }
  Border_1 {
   widgetClassPath=&Border
   parent=SizeBox_0
   widget=Border_1
   properties {brushColor=<rgba 0.0075 0.0075 0.0075 1>, padding=<ltrb 2 2 2 2>}
   slot {class=SizeBoxSlot}
  }
  Border_0 {
   widgetClassPath=&Border
   parent=Border_1
   widget=Border_0
   properties {brushColor=<rgba 0.3281 0.3281 0.3281 1>}
   slot {class=BorderSlot, properties {padding=<ltrb 2 2 2 2>}}
  }
  Text_Value {
   widgetClassPath=&CommonTextBlock
   bIsVariable=true
   parent=Border_0
   widget=Text_Value
   properties {
    autoWrapText=true
    text=Text Block
    font {
     bForceMonospaced=false
     bMaterialIsStencil=false
     fontMaterial=null
     fontObject=&NotoSans
     letterSpacing=0
     monospacedWidth=1
     size=11
     skewAmount=0
     typefaceFontName=NotoSans-Regular
     outlineSettings {
      bApplyOutlineToDropShadows=false
      bMiteredCorners=false
      bSeparateFillAlpha=false
      outlineMaterial=null
      outlineSize=0
      outlineColor=<rgba 0 0 0 1>
     }
    }
    shadowOffset=<0 0>
   }
   slot {class=BorderSlot}
  }
 ]
}
