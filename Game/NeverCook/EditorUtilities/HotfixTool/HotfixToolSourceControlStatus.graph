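# Editor utility widget for the Hotfix Tool that warns when source control is not connected, which the tool needs to determine correct syntax.
# A horizontal box inherited from the native SourceControlStatusWidget parent holds a red-tinted warning icon and the warning text.
# The event graph has only unconnected PreConstruct, Construct and Tick events, so this asset supplies layout and adds no logic of its own.
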EditorUtilityWidgetBlueprint /Game/NeverCook/EditorUtilities/HotfixTool/HotfixToolSourceControlStatus
graphs [EventGraph {nodes [
   Event PreConstruct()
   Event Construct()
   Event Tick()
  ]}]
variables []
widgets {
 info {
  inheritedWidgetCount=1
  parentClass=/Script/FortniteEditor.SourceControlStatusWidget
  rootWidgetClass=&HorizontalBox
  widgetCount=3
 }
 widgets [
  Box_SourceControlWarning {widgetClassPath=&HorizontalBox, bInherited=true, widget=Box_SourceControlWarning}
  Image_Warning {
   widgetClassPath=&Image
   bIsVariable=true
   parent=Box_SourceControlWarning
   widget=Image_Warning
   properties {brush {
     bIsDynamicallyLoaded=false
     drawAs=Image
     imageType=NoImage
     mirroring=NoMirror
     resourceName=None
     resourceObject=/Game/UI/Foundation/Shell/Textures/T-Icon-Warning-64
     tiling=NoTile
     imageSize=<32 32>
     margin=<ltrb 0 0 0 0>
     outlineSettings {
      bUseBrushTransparency=false
      roundingType=HalfHeightRadius
      width=0
      color {colorUseRule=UseColor_Specified, specifiedColor=<rgba 0 0 0 0>}
      cornerRadii=<0 0 0 0>
     }
     tintColor {colorUseRule=UseColor_Specified, specifiedColor=<rgba 0.9115 0.0253 0.0283 1>}
     uVRegion {bIsValid=false, max=<0 0>, min=<0 0>}
    }}
   slot {class=HorizontalBoxSlot, properties {horizontalAlignment=HAlign_Center, verticalAlignment=VAlign_Center, padding=<ltrb 5 5 5 5>}}
  }
  Text_SourceControlWarning {
   widgetClassPath=&TextBlock
   parent=Box_SourceControlWarning
   widget=Text_SourceControlWarning
   properties {
    text="Not connected to source control. Please Connect Source Control in order for the Hotfix Tool to determine correct syntax. "
    font {
     bForceMonospaced=false
     bMaterialIsStencil=false
     fontMaterial=null
     fontObject=&NotoSans
     letterSpacing=0
     monospacedWidth=1
     size=10.5
     skewAmount=0
     typefaceFontName=NotoSans-Regular
     outlineSettings {
      bApplyOutlineToDropShadows=false
      bMiteredCorners=false
      bSeparateFillAlpha=false
      outlineMaterial=null
      outlineSize=0
      outlineColor=<rgba 0 0 0 1>
     }
    }
   }
   slot {class=HorizontalBoxSlot, properties {horizontalAlignment=HAlign_Center, verticalAlignment=VAlign_Center, padding=<ltrb 4 2 4 2>}}
  }
 ]
}
