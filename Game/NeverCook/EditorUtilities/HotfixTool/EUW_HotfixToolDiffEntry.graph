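# Editor utility widget that displays one diff row in the NeverCook hotfix tool, showing a row name, column name, old value and new value.
# A horizontal box holds four fixed-size boxes, each wrapping a Common Text block as a column.
# InitHotfixDiffEntry takes the four strings and writes them into those text blocks.
# The event graph has no logic beyond empty stub events.
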EditorUtilityWidgetBlueprint /Game/NeverCook/EditorUtilities/HotfixTool/EUW_HotfixToolDiffEntry
graphs [
 EventGraph {nodes [
   Event PreConstruct()
   Event Construct()
   Event Tick()
  ]}
 InitHotfixDiffEntry {nodes [
   n1 = InitHotfixDiffEntry!() -> (then, RowName:String, ColumnName:String, OldValue:String, NewValue:String) {is=title K2Node_FunctionEntry}
   > Set Text(Text="To Text (String)"(InString=@n1), self=Get Text_RowName() -> (Text_RowName:Common Text Object Reference))
   > Set Text(Text="To Text (String)"(InString=@n1.ColumnName), self=Get Text_ColumnName() -> (Text_ColumnName:Common Text Object Reference))
   > Set Text(Text="To Text (String)"(InString=@n1.OldValue), self=Get Text_OldValue() -> (Text_OldValue:Common Text Object Reference))
   > Set Text(Text="To Text (String)"(InString=@n1.NewValue), self=Get Text_NewValue() -> (Text_NewValue:Common Text Object Reference))
  ]}
]
variables []
widgets {
 info {rootWidgetClass=&HorizontalBox, widgetCount=9}
 widgets [
  HorizontalBox_19 {widgetClassPath=&HorizontalBox, widget=HorizontalBox_19}
  SizeBox_0 {
   widgetClassPath=&SizeBox
   parent=HorizontalBox_19
   widget=SizeBox_0
   properties {bOverride_HeightOverride=true, bOverride_WidthOverride=true, heightOverride=30, widthOverride=300}
   slot {class=HorizontalBoxSlot, properties {horizontalAlignment=HAlign_Left, verticalAlignment=VAlign_Top, padding=<ltrb 0 0 10 0>}}
  }
  Text_RowName {
   widgetClassPath=&CommonTextBlock
   bIsVariable=true
   parent=SizeBox_0
   widget=Text_RowName
   properties {
    autoWrapText=true
    text=Text Block
    wrapTextAt=290
    colorAndOpacity {colorUseRule=UseColor_Specified, specifiedColor=<rgba 0.5841 0.5841 0.5841 1>}
    font {
     bForceMonospaced=false
     bMaterialIsStencil=false
     fontMaterial=null
     fontObject=&NotoSans
     letterSpacing=0
     monospacedWidth=1
     size=11
     skewAmount=0
     typefaceFontName=NotoSans-Regular
     outlineSettings {
      bApplyOutlineToDropShadows=false
      bMiteredCorners=false
      bSeparateFillAlpha=false
      outlineMaterial=null
      outlineSize=0
      outlineColor=<rgba 0 0 0 1>
     }
    }
    shadowOffset=<0 0>
   }
   slot {class=SizeBoxSlot}
  }
  SizeBox_1 {
   widgetClassPath=&SizeBox
   parent=HorizontalBox_19
   widget=SizeBox_1
   properties {bOverride_HeightOverride=true, bOverride_WidthOverride=true, heightOverride=30, widthOverride=300}
   slot {class=HorizontalBoxSlot, properties {horizontalAlignment=HAlign_Left, verticalAlignment=VAlign_Top, padding=<ltrb 0 0 10 0>}}
  }
  Text_ColumnName {
   widgetClassPath=&CommonTextBlock
   bIsVariable=true
   parent=SizeBox_1
   widget=Text_ColumnName
   properties {
    autoWrapText=true
    text=Text Block
    colorAndOpacity {colorUseRule=UseColor_Specified, specifiedColor=<rgba 0.5841 0.5841 0.5841 1>}
    font {
     bForceMonospaced=false
     bMaterialIsStencil=false
     fontMaterial=null
     fontObject=&NotoSans
     letterSpacing=0
     monospacedWidth=1
     size=11
     skewAmount=0
     typefaceFontName=NotoSans-Regular
     outlineSettings {
      bApplyOutlineToDropShadows=false
      bMiteredCorners=false
      bSeparateFillAlpha=false
      outlineMaterial=null
      outlineSize=0
      outlineColor=<rgba 0 0 0 1>
     }
    }
    shadowOffset=<0 0>
   }
   slot {class=SizeBoxSlot}
  }
  SizeBox_2 {
   widgetClassPath=&SizeBox
   parent=HorizontalBox_19
   widget=SizeBox_2
   properties {bOverride_HeightOverride=true, bOverride_WidthOverride=true, heightOverride=30, widthOverride=300}
   slot {class=HorizontalBoxSlot, properties {horizontalAlignment=HAlign_Left, verticalAlignment=VAlign_Top, padding=<ltrb 0 0 10 0>}}
  }
  Text_OldValue {
   widgetClassPath=&CommonTextBlock
   bIsVariable=true
   parent=SizeBox_2
   widget=Text_OldValue
   properties {
    autoWrapText=true
    text=Text Block
    colorAndOpacity {colorUseRule=UseColor_Specified, specifiedColor=<rgba 0.5841 0.5841 0.5841 1>}
    font {
     bForceMonospaced=false
     bMaterialIsStencil=false
     fontMaterial=null
     fontObject=&NotoSans
     letterSpacing=0
     monospacedWidth=1
     size=11
     skewAmount=0
     typefaceFontName=NotoSans-Regular
     outlineSettings {
      bApplyOutlineToDropShadows=false
      bMiteredCorners=false
      bSeparateFillAlpha=false
      outlineMaterial=null
      outlineSize=0
      outlineColor=<rgba 0 0 0 1>
     }
    }
    shadowOffset=<0 0>
   }
   slot {class=SizeBoxSlot}
  }
  SizeBox_3 {
   widgetClassPath=&SizeBox
   parent=HorizontalBox_19
   widget=SizeBox_3
   properties {bOverride_HeightOverride=true, bOverride_WidthOverride=true, heightOverride=30, widthOverride=300}
   slot {class=HorizontalBoxSlot, properties {horizontalAlignment=HAlign_Left, verticalAlignment=VAlign_Top}}
  }
  Text_NewValue {
   widgetClassPath=&CommonTextBlock
   bIsVariable=true
   parent=SizeBox_3
   widget=Text_NewValue
   properties {
    autoWrapText=true
    text=Text Block
    colorAndOpacity {colorUseRule=UseColor_Specified, specifiedColor=<rgba 0.5841 0.5841 0.5841 1>}
    font {
     bForceMonospaced=false
     bMaterialIsStencil=false
     fontMaterial=null
     fontObject=&NotoSans
     letterSpacing=0
     monospacedWidth=1
     size=11
     skewAmount=0
     typefaceFontName=NotoSans-Regular
     outlineSettings {
      bApplyOutlineToDropShadows=false
      bMiteredCorners=false
      bSeparateFillAlpha=false
      outlineMaterial=null
      outlineSize=0
      outlineColor=<rgba 0 0 0 1>
     }
    }
    shadowOffset=<0 0>
   }
   slot {class=SizeBoxSlot}
  }
 ]
}
